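# Animation Blueprint for the M_MED_Orbit_Teal cosmetic body (Switch LOD1 cloth variant), adding secondary motion to the parent mesh's copied pose.
# Two deformation Control Rigs feed gauntlet offsets, weighted constraints and look-ats for right-arm tubes, and RigidBody sims for tubes, flaps and holster.
# Event graph drives wrist and elbow alphas from bone measurements and runs a deprecated nativized cloth controller.
# In most vehicle seats, cloth blend drops to zero and back loincloth bones shrink.
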AnimBlueprint /BRCosmetics/Characters/Player/Male/Medium/Bodies/M_MED_Orbit_Teal/Meshes/ClothForSwitch/M_MED_Orbit_Teal_AnimBP_ForSwitchLOD1
target_skeleton=/BRCosmetics/Characters/Player/Male/Medium/Bodies/M_MED_Orbit_Teal/Meshes/M_MED_Orbit_Teal_Skeleton
graphs [
 AnimGraph {nodes [
   n1 = Control Rig(Source=Copy Pose From Mesh() {bUseAttachedParent=true}) {
    alphaBoolBlend {blendOption=Linear}
    controlRigAssetReference {blueprintRigClass=/BRCosmetics/Characters/Player/Male/Medium/Bodies/M_MED_Orbit_Teal/Meshes/M_MED_Orbit_Teal_Deform_CtrlRig_twists.M_MED_Orbit_Teal_Deform_CtrlRig_twists_C}
   }
   n2 = Control Rig!(
    Source=@n1
    Alpha
    KneeRot_fr_l_114deg:Vector
    KneeTrans_fr_l_114deg:Vector="0.258,-3.591,0.009"
    KneeScale_fr_l_114deg:Vector="1,1,1"
    KneeScale_fr_l_130deg:Vector="1,1,1"
    KneeRot_fr_l_130deg:Vector="0,0,-51.2"
    KneeTrans_fr_l_130deg:Vector="0.258,-3.591,0.009"
   ) -> (Pose) {
    alphaBoolBlend {blendOption=Linear}
    controlRigAssetReference {blueprintRigClass=/BRCosmetics/Characters/Player/Male/Medium/Bodies/M_MED_Orbit_Teal/Meshes/M_MED_Orbit_Teal_Deform_CtrlRig_sculpters.M_MED_Orbit_Teal_Deform_CtrlRig_sculpters_C}
   }
   n3 = "Transform (Modify) Bone - Bone: dyn_gauntlet_l"(
    ComponentPose=Local To Component(LocalPose=@n2)
    Alpha=Get wrist_l_rot() -> (wrist_l_rot:double)
    Translation="-4.5,0,0"
    Rotation="0,0,0"
    Scale="1,1,1"
   ) {
    rotationSpace=BCS_BoneSpace
    translationMode=BMM_Additive
    translationSpace=BCS_BoneSpace
    alphaBoolBlend {blendOption=Linear}
    alphaScaleBiasClamp {bInterpResult=true, interpSpeedDecreasing=20, interpSpeedIncreasing=20}
    boneToModify {boneName=dyn_gauntlet_l}
    translation {x=-4.5}
   }
   n4 = "Transform (Modify) Bone - Bone: dyn_gauntlet_r"(ComponentPose=@n3, Alpha=Get wrist_r_rot() -> (wrist_r_rot:double), Translation="-4.5,0,0", Rotation="0,0,0", Scale="1,1,1") {
    rotationSpace=BCS_BoneSpace
    translationMode=BMM_Additive
    translationSpace=BCS_BoneSpace
    alphaBoolBlend {blendOption=Linear}
    alphaScaleBiasClamp {bInterpResult=true, interpSpeedDecreasing=20, interpSpeedIncreasing=20}
    boneToModify {boneName=dyn_gauntlet_r}
    translation {x=-4.5}
   }
   n5 = Get in vehicle scale alpha() -> (in vehicle scale alpha:double)
   n6 = "Transform (Modify) Bone - Bone: dyn_loincloth_bk_l_02"(ComponentPose=@n4, Alpha=@n5, Translation="0,0,0", Rotation="0,0,0", Scale="0.1,0.5,1") {scaleMode=BMM_Replace, alphaBoolBlend {blendOption=Linear}, boneToModify {boneName=dyn_loincloth_bk_l_02}, scale=<0.1 0.5>}
   n7 = "Transform (Modify) Bone - Bone: dyn_loincloth_bk_02"(ComponentPose=@n6, Alpha=@n5, Translation="0,0,0", Rotation="0,0,0", Scale="0.1,0.5,1") {scaleMode=BMM_Replace, alphaBoolBlend {blendOption=Linear}, boneToModify {boneName=dyn_loincloth_bk_02}, scale=<0.1 0.5>}
   n8 = "Transform (Modify) Bone - Bone: dyn_roincloth_bk_r_02"(ComponentPose=@n7, Alpha=@n5, Translation="0,0,0", Rotation="0,0,0", Scale="0.1,0.5,1") {scaleMode=BMM_Replace, alphaBoolBlend {blendOption=Linear}, boneToModify {boneName=dyn_roincloth_bk_r_02}, scale=<0.1 0.5>}
   n9 = "Constraint - dyn_tubes_top_r_MSTR (PPPPPPPP)"!(
    ComponentPose=@n8
    Alpha=1
    ConstraintWeights_0=0.3538
    ConstraintWeights_1:float=0.2792
    ConstraintWeights_2:float=0.1058
    ConstraintWeights_3:float=0.1032
    ConstraintWeights_4:float=0.0545
    ConstraintWeights_5:float=0.0411
    ConstraintWeights_6:float=0.0322
    ConstraintWeights_7:float=0.0302
   ) -> (Pose) {
    alphaBoolBlend {blendOption=Linear}
    boneToModify {boneName=dyn_tubes_top_r_MSTR}
    constraintSetup [
     {targetBone {boneName=upperarm_twist_01_r}}
     {targetBone {boneName=upperarm_twist_02_r}}
     {targetBone {boneName=clavicle_r}}
     {targetBone {boneName=deform_deltoid_fr_r}}
     {targetBone {boneName=spine_05}}
     {targetBone {boneName=deform_pectoral_fr_r}}
     {targetBone {boneName=upperarm_r}}
     {targetBone {boneName=spine_04}}
    ]
    constraintWeights [1,1,1,1,1,1,1,1]
   }
   n10 = "Constraint - dyn_tubes_bott_r_MSTR (PPPPPP)"!(
    ComponentPose=@n9
    Alpha=1
    ConstraintWeights_0=0.4886
    ConstraintWeights_1:float=0.2817
    ConstraintWeights_2:float=0.112
    ConstraintWeights_3:float=0.056
    ConstraintWeights_4:float=0.0342
    ConstraintWeights_5:float=0.0275
   ) -> (Pose) {
    alphaBoolBlend {blendOption=Linear}
    boneToModify {boneName=dyn_tubes_bott_r_MSTR}
    constraintSetup [
     {targetBone {boneName=upperarm_twist_02_r}}
     {targetBone {boneName=upperarm_twist_01_r}}
     {targetBone {boneName=clavicle_r}}
     {targetBone {boneName=upperarm_r}}
     {targetBone {boneName=deform_deltoid_fr_r}}
     {targetBone {boneName=spine_05}}
    ]
    constraintWeights [1,1,1,1,1,1]
   }
   n11 = "Constraint - dyn_tubes_bott_r_02_MSTR (PP)"!(ComponentPose=@n10, Alpha=1, ConstraintWeights_0=0.4, ConstraintWeights_1:float=0.6) -> (Pose) {
    alphaBoolBlend {blendOption=Linear}
    boneToModify {boneName=dyn_tubes_bott_r_02_MSTR}
    constraintSetup [{targetBone {boneName=lowerarm_twist_02_r}}, {targetBone {boneName=upperarm_r}}]
    constraintWeights [1,1]
   }
   n12 = "Constraint - dyn_tubes_top_r_02_MSTR (PP)"!(ComponentPose=@n11, Alpha=1, ConstraintWeights_0=0.4, ConstraintWeights_1:float=0.6) -> (Pose) {
    alphaBoolBlend {blendOption=Linear}
    boneToModify {boneName=dyn_tubes_top_r_02_MSTR}
    constraintSetup [{targetBone {boneName=lowerarm_twist_02_r}}, {targetBone {boneName=upperarm_r}}]
    constraintWeights [1,1]
   }
   n13 = Get elbow_rot_r() -> (elbow_rot_r:double)
   n14 = Weighted Look At - Bone: dyn_tubes_top_r_MSTR!(ComponentPose:Component Space Pose Link Structure=@n12, Alpha:float=@n13) -> (Pose:Component Space Pose Link Structure) {
    is=title FortAnimGraphNode_WeightedLookAt
    bUseLookUpAxis=true
    alphaBoolBlend {blendOption=Linear}
    boneToModify {boneName=dyn_tubes_top_r_MSTR}
    lookAtLocation {x=0}
    lookAtTarget {boneReference {boneName=dyn_tubes_top_r_02_MSTR}}
    lookAt_Axis {axis=<1 0>}
    lookUp_Axis {axis {x=0, z=1}}
   }
   n15 = Weighted Look At - Bone: dyn_tubes_bott_r_MSTR!(ComponentPose:Component Space Pose Link Structure=@n14, Alpha:float=1) -> (Pose:Component Space Pose Link Structure) {
    is=title FortAnimGraphNode_WeightedLookAt
    bUseLookUpAxis=true
    alphaBoolBlend {blendOption=Linear}
    boneToModify {boneName=dyn_tubes_bott_r_MSTR}
    lookAtLocation {x=0}
    lookAtTarget {boneReference {boneName=dyn_tubes_bott_r_02_MSTR}}
    lookAt_Axis {axis=<1 0>}
    lookUp_Axis {axis {x=0, z=1}}
   }
   n16 = Weighted Look At - Bone: dyn_tubes_top_r_02_MSTR!(ComponentPose:Component Space Pose Link Structure=@n15, Alpha:float=@n13) -> (Pose:Component Space Pose Link Structure) {
    is=title FortAnimGraphNode_WeightedLookAt
    bUseLookUpAxis=true
    alphaBoolBlend {blendOption=Linear}
    boneToModify {boneName=dyn_tubes_top_r_02_MSTR}
    lookAtLocation {x=0}
    lookAtTarget {boneReference {boneName=dyn_tubes_top_r_MSTR}}
    lookAt_Axis {axis=<1 0>}
    lookUp_Axis {axis {x=0, z=1}}
   }
   n17 = Weighted Look At - Bone: dyn_tubes_bott_r_02_MSTR!(ComponentPose:Component Space Pose Link Structure=@n16, Alpha:float=1) -> (Pose:Component Space Pose Link Structure) {
    is=title FortAnimGraphNode_WeightedLookAt
    bUseLookUpAxis=true
    alphaBoolBlend {blendOption=Linear}
    boneToModify {boneName=dyn_tubes_bott_r_02_MSTR}
    lookAtLocation {x=0}
    lookAtTarget {boneReference {boneName=dyn_tubes_bott_r_MSTR}}
    lookAt_Axis {axis=<1 0>}
    lookUp_Axis {axis {x=0, z=1}}
   }
   n18 = RigidBody!(ComponentPose=@n17, Alpha, OverrideWorldGravity:Vector="0,0,-980", ExternalForce) -> (Pose) {
    bClampLinearTranslationLimitToRefPose=true
    bOverrideWorldGravity=true
    overridePhysicsAsset=/BRCosmetics/Characters/Player/Male/Medium/Bodies/M_MED_Orbit_Teal/Meshes/M_MED_Orbit_Teal_Dyn_tubes
    simulationSpace=BaseBoneSpace
    alphaBoolBlend {blendOption=Linear}
    baseBoneRef {boneName=upperarm_twist_01_r}
    componentAppliedLinearAccClamp=<3000 3000 3000>
    overrideWorldGravity {z=-980}
    simSpaceSettings {
     maxAngularAcceleration=1000
     maxAngularVelocity=3
     maxLinearAcceleration=250
     maxLinearVelocity=500
     velocityScaleZ=0
     worldAlpha=1
     externalLinearDragV {y=0.5}
    }
   }
   n19 = RigidBody(ComponentPose=@n18, OverrideWorldGravity="0,0,-980") {
    bClampLinearTranslationLimitToRefPose=true
    bOverrideWorldGravity=true
    overridePhysicsAsset=/BRCosmetics/Characters/Player/Male/Medium/Bodies/M_MED_Orbit_Teal/Meshes/M_MED_Orbit_Teal_Dyn_flaps
    simulationSpace=BaseBoneSpace
    alphaBoolBlend {blendOption=Linear}
    baseBoneRef {boneName=spine_05}
    componentAppliedLinearAccClamp=<3000 3000 3000>
    overrideWorldGravity {z=-980}
    simSpaceSettings {
     maxAngularAcceleration=1000
     maxAngularVelocity=3
     maxLinearAcceleration=250
     maxLinearVelocity=500
     velocityScaleZ=0
     externalLinearDragV {y=0.5}
    }
   }
   n20 = "Constraint - dyn_holster_bott_MSTR (PPPPPPPP)"!(
    ComponentPose=@n19
    Alpha=1
    ConstraintWeights_0=0.5357
    ConstraintWeights_1:float=0.3307
    ConstraintWeights_2:float=0.0468
    ConstraintWeights_3:float=0.041
    ConstraintWeights_4:float=0.0196
    ConstraintWeights_5:float=0.014
    ConstraintWeights_6:float=0.0112
    ConstraintWeights_7:float=0.001
   ) -> (Pose) {
    alphaBoolBlend {blendOption=Linear}
    boneToModify {boneName=dyn_holster_bott_MSTR}
    constraintSetup [
     {targetBone {boneName=thigh_r}}
     {targetBone {boneName=thigh_twist_01_r}}
     {targetBone {boneName=deform_knee_fr_01_r}}
     {targetBone {boneName=deform_knee_bk_01_r}}
     {targetBone {boneName=pelvis}}
     {targetBone {boneName=deform_thigh_fr_r}}
     {targetBone {boneName=calf_r}}
     {targetBone {boneName=deform_thigh_bk_r}}
    ]
    constraintWeights [1,1,1,1,1,1,1,1]
   }
   n21 = RigidBody(ComponentPose=@n20, OverrideWorldGravity="0,0,-980") {
    bClampLinearTranslationLimitToRefPose=true
    bOverrideWorldGravity=true
    overridePhysicsAsset=/BRCosmetics/Characters/Player/Male/Medium/Bodies/M_MED_Orbit_Teal/Meshes/M_MED_Orbit_Teal_Dyn_holster
    simulationSpace=BaseBoneSpace
    alphaBoolBlend {blendOption=Linear}
    baseBoneRef {boneName=thigh_twist_01_r}
    componentAppliedLinearAccClamp=<3000 3000 3000>
    overrideWorldGravity {z=-980}
    simSpaceSettings {
     maxAngularAcceleration=1000
     maxAngularVelocity=3
     maxLinearAcceleration=250
     maxLinearVelocity=500
     velocityScaleZ=0
     worldAlpha=1
     externalLinearDragV {y=0.5}
    }
   }
   Output Pose(Result=Component To Local(ComponentPose=@n21)) {layerGroup=None}
  ]}
 EventGraph {nodes [
   n1 = Event BlueprintUpdateAnimation()
   > Set DeltaTimeLocal(DeltaTimeLocal:double=@n1.DeltaTimeX) -> (Output_Get:double)
   > n2 = Sequence!(execute) -> (then_0, then_1, then_2:Exec, then_3:Exec, then_4:Exec, then_5:Exec)
   > n3 = "Cloth Controller Function (Parameter Asset) (Nativized) DEPRECATED"!(
    execute
    CharacterPartAnimInstance:Custom Character Part Anim Instance Object Reference=Self-Reference()
    ClothingAssetName:Name=M_MED_Orbit_Teal_Clothing_0
    ClothCoreSettings:"Cloth Core Settings Structure (by ref)"=Get Cloth Core Settings Native() -> (Cloth Core Settings Native:Cloth Core Settings Structure)
    WindGustLocalSettings:"Wind Gust Settings Structure (by ref)"=Get Wind Gust Local Settings Native() -> (Wind Gust Local Settings Native:Wind Gust Settings Structure)
    ClothUpAxis:Vector="1,0,0"
    OnGroundMovingSpeedLimit:float=2
    OnGroundMovingAngularVelMax:float=0.05
    OnGroundMovingFictitiousVelMax:float=0.6
    LinVelScaleFwdBwdClamp:float=0
    LinVelScaleUpDwnClamp:float=0.7
    DeltaTime:float=Get DeltaTimeLocal() -> (DeltaTimeLocal:double)
    GameTimeInSeconds:float=GetGameTimeInSeconds()
    StateLogicLayerName:Name=none
    Parameters:Fort Character Dynamics Parameters Object Reference=/BRCosmetics/Characters/Player/Male/Medium/Bodies/M_MED_Orbit_Teal/Meshes/M_MED_Orbit_Teal_Parameters
   ) -> (
    then
    OutCurrentState:Name=None
    OutClothParameters:Cloth Parameters Structure
    OutClothCoreSettings:Cloth Core Settings Structure
    OutWindGustSettings:Wind Gust Parameters Structure
    OutWindGustLocalSettings:Wind Gust Settings Structure
    OutWindVelocity:Vector
    OutLinearVelVector:Vector
   )
   > Set Wind Gust Local Settings Native Out(Wind Gust Local Settings Native Out:Wind Gust Settings Structure=@n3.OutWindGustLocalSettings) -> (Output_Get:Wind Gust Settings Structure)
   n4 = GetPartSkeletalMeshComponent()
   Set wrist_l_rot(
    execute=@n2.then_1
    wrist_l_rot:double=Get Distance Between Two Sockets!(
     Component:Skeletal Mesh Component Object Reference=@n4
     SocketOrBoneNameA:Name=dyn_hand_loc_tracker_l
     SocketSpaceA:ERelativeTransformSpace Enum=RTS_World
     SocketOrBoneNameB:Name=index_01_l
     SocketSpaceB:ERelativeTransformSpace Enum=RTS_World
     bRemapRange:bool=true
     InRangeMin:float=-5
     InRangeMax:float=9.38
     OutRangeMin:float=1
     OutRangeMax:float=0
    ) -> (ReturnValue:float) {is=title K2Node_CallFunction}
   ) -> (Output_Get:double)
   > Set wrist_r_rot(wrist_r_rot:double=Get Distance Between Two Sockets(
     Component=@n4
     SocketOrBoneNameA=dyn_hand_loc_tracker_r
     SocketSpaceA=RTS_World
     SocketOrBoneNameB=index_01_r
     SocketSpaceB=RTS_World
     bRemapRange=true
     InRangeMin=0
     InRangeMax=9.38
     OutRangeMin=1
     OutRangeMax=0
    )) -> (Output_Get:double)
   Set elbow_rot_r(
    execute=@n2.then_2
    elbow_rot_r:double=GetAlphaFromReferenceTransformDelta(SkeletalMeshComponent=@n4, BoneName=lowerarm_r, BaseName=upperarm_r, InputRange=(X=0,Y=80), OutputRange=(X=1,Y=0), TwistAxis="1,0,0")
   ) -> (Output_Get:double)
   > Set upperarm_rotZ_r(upperarm_rotZ_r:double=GetAlphaFromReferenceTransformDelta(SkeletalMeshComponent=@n4, BoneName=upperarm_r, BaseName=clavicle_r, DeltaComponentType=EulerY, InputRange=(X=-20,Y=20), TwistAxis="0,0,1")) -> (Output_Get:double)
   > Set upperarm_rotZ_l(upperarm_rotZ_l:double=GetAlphaFromReferenceTransformDelta(SkeletalMeshComponent=@n4, BoneName=upperarm_l, BaseName=clavicle_l, DeltaComponentType=EulerY, InputRange=(X=-5,Y=20), TwistAxis="0,0,1")) -> (Output_Get:double)
   n5 = Branch(execute=@n2.then_5, Condition=Get bIsInVehicle() -> (bIsInVehicle:bool))
   n6 = Get bVehicleSeatForcesCrouch() -> (bVehicleSeatForcesCrouch:bool)
   n7 = Get bIsGolfCartVehicleBackPassenger() -> (bIsGolfCartVehicleBackPassenger:bool)
   n8 = Get bIsJackalVehicleDriver() -> (bIsJackalVehicleDriver:bool)
   n9 = Get bIsInShoppingCart() -> (bIsInShoppingCart:bool)
   n10 = Get bIsInCannon() -> (bIsInCannon:bool)
   n11 = Get bIsMeatballPassenger() -> (bIsMeatballPassenger:bool)
   > n12 = Branch(Condition=NOT Boolean(A=OR Boolean!(A=@n6, B=@n7, C:bool=@n8, D:bool=@n9, E:bool=@n10, F:bool=@n11) -> (ReturnValue)))
   > Set in vehicle rigid alpha(in vehicle rigid alpha:double=0) -> (Output_Get:double)
   > Set in vehicle scale alpha(in vehicle scale alpha:double=1) -> (Output_Get:double)
   n13 = GetPartSkeletalMeshComponent()
   > Set ClothBlendWeight!(execute, ClothBlendWeight:float=0, self:Skeletal Mesh Component Object Reference=@n13) -> (then, Output_Get:float)
   Set in vehicle rigid alpha(execute=[@n12.else, @n5.else], in vehicle rigid alpha=1)
   > Set in vehicle scale alpha(in vehicle scale alpha=0)
   > Set ClothBlendWeight(ClothBlendWeight=1, self=@n13)
  ]}
]
variables [
 Cloth Core Settings Native {
  type=ClothCoreSettings
  default {
   collisionThickness=0
   dampingCoefficient=0.1
   frictionCoefficient=0.3
   localDampingCoefficient=0.1
   numberIterations=2
   numberSubsteps=1
   selfCollisionThickness=0
   animDriveDamping=<0 1>
   animDriveStiffness=<0 1>
   areaStiffness=<1 1>
   bendingStiffness=<1 1>
   edgeStiffness=<1 1>
   tetherStiffness=<1 1>
  }
 }
 DeltaTimeLocal {type=number}
 elbow_rot_r {type=number}
 in vehicle rigid alpha {type=number}
 in vehicle scale alpha {type=number}
 upperarm_rotZ_l {type=number}
 upperarm_rotZ_r {type=number}
 Wind Gust Local Settings Native {type=WindGustSettings, default {elapsedTime=0, finalGustStrengthLocal=0, gustDuration=0, gustStrength=0, timeBetweenGusts=0}}
 Wind Gust Local Settings Native Out {type=WindGustSettings, default {elapsedTime=0, finalGustStrengthLocal=0, gustDuration=0, gustStrength=0, timeBetweenGusts=0}}
 wrist_l_rot {type=number}
 wrist_r_rot {type=number}
]
